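; Discoverable quest actor for locating the map in the bvg maproom quest (s12 special).
; After playlist data loads, it checks Creative mode; the server binds TouchBox overlaps while clients cache the objective count and listen for quest updates.
; Outside Creative, touches by players with the unfinished quest send a map-visit stat event and call a client-only delayed particle.
; An increased count triggers the child success hook, then optional delayed sound and FX.
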(event EventBeginPlay
  (bind _returnvalue (Fortnite|GetLocalFortPlayerControllers true))
  (bind _output (Utilities|Array|Get(acopy) _returnvalue))
  (bind _touchbox (Variables|Default|GetTouchBox))
  ;[1 0.4483 0.1218] STILL TEMP!!!
  (Utilities|FlowControl|Delay 5.0)
  ;[1 1 1] CLIENT AND SERVER - Check if In Creative Mode or not
  (bind _asfort_game_state_athena (Utilities|Casting|CastToFortGameStateAthena (Game|GetGameState)))
  (if (Variables|/Script/FortniteGame.FortGameStateAthena|GetbPlaylistDataIsLoaded _asfort_game_state_athena)
    (CallFunction|CheckCreativeMode)
    ;[1 1 1] Check for Quest and Objective
    (Networking|SwitchHasAuthority
      (:Authority
        ;[1 1 1] Bind Overlap Event
        (Collision|BindEventtoOnComponentBeginOverlap _touchbox (EventDispatchers|CreateEvent :!event Custom|TouchedQuestActor))
        (Development|PrintString "BOUND EVENT" false false)
        (Collision|BindEventtoOnComponentEndOverlap _touchbox (EventDispatchers|CreateEvent :!event Custom|QuestActor_EndOverlap)))
      (:Remote
        (Variables|Default|SetIsLocallyValid true)
        ;[1 1 1] Bind Events for Touch and QuestUpdated
        (bind _questcountachieved (Class|AthenaQuestFunctionLibrary|AthenaQuestObjectiveCounter _output (Variables|Default|GetQuestItem)))
        (bind _output_get (Variables|Default|SetPlayerObjectiveCount _questcountachieved))
        (Development|PrintString (Utilities|String|ToString(Integer) _output_get) (Variables|Default|GetbTestQuest) (Variables|Default|GetbTestQuest))
        (Quests|BindEventtoOnQuestsUpdated (Quests|GetQuestManager _output "Invalid") (EventDispatchers|CreateEvent :!event Custom|QuestSuccessfullyUpdated))
        ;[1 1 1] Tell Children of
        (CallFunction|QuestInitializedandValid)))
    (else
      (Playlist|BindEventtoOnCurrentPlaylistLoaded _asfort_game_state_athena (delegate Custom|PlaylistLoaded)))))

(event Custom|PlaylistLoaded
  (params (PlaylistName :type Name) (PlaylistContextTags :type Gameplay%20Tag%20Container%20Structure%20%28by%20ref%29))
  (bind _returnvalue (Fortnite|GetLocalFortPlayerControllers true))
  (bind _output (Utilities|Array|Get(acopy) _returnvalue))
  (bind _touchbox (Variables|Default|GetTouchBox))
  ;[1 1 1] CLIENT AND SERVER - Check if In Creative Mode or not
  (CallFunction|CheckCreativeMode)
  ;[1 1 1] Check for Quest and Objective
  (Networking|SwitchHasAuthority
    (:Authority
      ;[1 1 1] Bind Overlap Event
      (Collision|BindEventtoOnComponentBeginOverlap _touchbox (EventDispatchers|CreateEvent :!event Custom|TouchedQuestActor))
      (Development|PrintString "BOUND EVENT" false false)
      (Collision|BindEventtoOnComponentEndOverlap _touchbox (EventDispatchers|CreateEvent :!event Custom|QuestActor_EndOverlap)))
    (:Remote
      (Variables|Default|SetIsLocallyValid true)
      ;[1 1 1] Bind Events for Touch and QuestUpdated
      (bind _questcountachieved (Class|AthenaQuestFunctionLibrary|AthenaQuestObjectiveCounter _output (Variables|Default|GetQuestItem)))
      (bind _output_get (Variables|Default|SetPlayerObjectiveCount _questcountachieved))
      (Development|PrintString (Utilities|String|ToString(Integer) _output_get) (Variables|Default|GetbTestQuest) (Variables|Default|GetbTestQuest))
      (Quests|BindEventtoOnQuestsUpdated (Quests|GetQuestManager _output "Invalid") (EventDispatchers|CreateEvent :!event Custom|QuestSuccessfullyUpdated))
      ;[1 1 1] Tell Children of
      (CallFunction|QuestInitializedandValid))))

(event Custom|QuestInitializedAndValid)

(event Custom|TouchedQuestActor
  (params (OverlappedComponent :type Primitive%20Component%20Object%20Reference) (OtherActor :type Actor%20Object%20Reference) (OtherComp :type Primitive%20Component%20Object%20Reference) (OtherBodyIndex :type Integer) (bFromSweep :type Boolean) (SweepResult :type Hit%20Result%20Structure%20%28by%20ref%29))
  (bind _tagtocheck (Variables|Default|GetTagToCheck))
  (bind _btestquest (Variables|Default|GetbTestQuest))
  ;[1 1 1] Touched Capsule
  (if (Variables|Default|GetRequiresConditionToComplete)
    (bind _returnvalue (GameplayTags|HasMatchingGameplayTag(Message) OtherActor _tagtocheck))
    (if (not _returnvalue)
      (Development|PrintString (Utilities|String|Append "Missing gameplay tag: " (Utilities|String|ToString(Name) (GameplayTags|GetTagName _tagtocheck))) _btestquest _btestquest "(R=1.000000,G=0.021699,B=0.000000,A=1.000000)" 5.0))
    (else
      _))
  (bind _asfort_player_controller_athena (Utilities|Casting|CastToFortPlayerControllerAthena (Game|GetInstigatorController OtherActor))
    (:then
      (if (Variables|Default|GetCreativeMode)
        (CallFunction|TouchedActorinCreativeMode)
        (else
          (bind (_hasquest _questcompleted) (Class|AthenaQuestFunctionLibrary|AthenaHasPlayerCompletedQuest _asfort_player_controller_athena "/Game/Athena/Items/Quests/DiscoverableQuests/quest_s12_special_bvg_maproom.quest_s12_special_bvg_maproom"))
          (if _questcompleted
            _
            (elif _hasquest
              (Quest|SendComplexCustomStatEvent (Quests|GetQuestManager _asfort_player_controller_athena "Athena") 0 "(GameplayTags=((TagName=\"Athena.Location.Visit.Quest.BVG.Map\")))" "(GameplayTags=)" 1)
              (CallFunction|Particle))))))
    (:CastFailed
      (Development|PrintString "NOT A PLAYER" _btestquest _btestquest "(R=1.000000,G=0.021699,B=0.000000,A=1.000000)" 5.0))))

(event Custom|Particle
  ;[1 1 1] Touched Capsule
  (Networking|SwitchHasAuthority
    (:Remote
      (Utilities|FlowControl|Delay 0.9)
      (Effects|Components|ParticleSystem|SpawnFXAtLocation "/Game/Athena/Items/QuestInteractables/Generic/FX/NS_Athena_QuestItem_PickupSuccess.NS_Athena_QuestItem_PickupSuccess" (Transformation|GetWorldLocation (Variables|Default|GetDeco_AppleMap)) "0, 0,-90.000000"))))

(event Custom|QuestActor_EndOverlap
  (params (OverlappedComponent :type Primitive%20Component%20Object%20Reference) (OtherActor :type Actor%20Object%20Reference) (OtherComp :type Primitive%20Component%20Object%20Reference) (OtherBodyIndex :type Integer)))

(event Custom|QuestSuccessfullyUpdated
  ;[1 1 1] Quest Updated
  (if (Variables|Default|GetisLocallyValid)
    (bind _questcountachieved (Class|AthenaQuestFunctionLibrary|AthenaQuestObjectiveCounter (Utilities|Array|Get(acopy) (Fortnite|GetLocalFortPlayerControllers true)) (Variables|Default|GetQuestItem)))
    (if (> _questcountachieved (Variables|Default|GetPlayerObjectiveCount))
      (Variables|Default|SetIsLocallyValid false)
      (CallFunction|ObjectiveSuccessfullyCompleted (Variables|Default|GetPlayerThatTriggeredObject)))))

(event Custom|OnUsingFocalPointChanged
  (params (FocalPointUser :type Fort%20Player%20Controller%20Athena%20Object%20Reference) (FocalPointActor :type Actor%20Object%20Reference) (bUsingFocalPoint :type Boolean)))

(event Custom|ObjectiveSuccessfullyCompleted
  (params (PlayerWhoTriggeredObject :type Actor%20Object%20Reference))
  (bind _genericsuccesssound (Variables|Default|GetGenericSuccessSound))
  (bind _genericsuccessfx (Variables|Default|GetGenericSuccessFX))
  ;[1 1 1] Success FX and Sound
  (CallFunction|TellChildOfSuccess)
  (Utilities|FlowControl|Delay (Variables|Default|GetDelayBeforeGenericSuccessSoundAndFX))
  (Utilities|IsValid _genericsuccesssound
    (:Is%20Valid
      (Audio|SpawnSoundatLocation _genericsuccesssound (Transformation|GetActorLocation)))
    (:Is%20Not%20Valid
      _))
  (Utilities|IsValid _genericsuccessfx
    (:Is%20Valid
      (Effects|SpawnFXSystemAttached _genericsuccessfx (Variables|/Script/Engine.Character|GetMesh (Game|Player|GetPlayerPawn (Utilities|Array|Get(acopy) (Fortnite|GetLocalFortPlayerControllers true))))))))

(event Custom|TellChildOfSuccess)

(event Custom|TouchedActorInCreativeMode
  ;[1 1 1] Success FX and Sound
  (Networking|SwitchHasAuthority
    (:Remote
      (CallFunction|TellChildOfSuccess))))

(fn CheckCreativeMode
  (bind _cancompletequest (Class|AthenaQuestFunctionLibrary|AthenaCheckCreativeMode (Game|GetGameState)))
  (Variables|Default|SetCreativeMode (not _cancompletequest))
  (return))

(fn OnRep_ShoppingCartThatTriggered)

(fn ConstructionScript)
